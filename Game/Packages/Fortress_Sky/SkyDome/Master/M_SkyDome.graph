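# Unlit master sky dome material for Fortress_Sky, compositing a sun, horizon-to-zenith gradient, galaxy and star layers, and animated clouds.
# Clouds layer panning normal/height and detail textures through a swirl function that also rotates the normals.
# Cloud lighting uses the sun/moon position from a parameter collection for shadow, rim and bounce tints, while cloud opacity occludes sun glow and stars.
# Shading-path switches simplify clouds and drop stars and rimlight on mobile.
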Material /Game/Packages/Fortress_Sky/SkyDome/Master/M_SkyDome
graph {
 properties {bTangentSpaceNormal=false, pixelDepthOffsetMode=PDOM_Legacy, shadingModel=MSM_Unlit, translucencyPass=MTP_BeforeDOF, usedShadingModels=Unlit}
 nodes [
  n1 = Power(
   Base=ComponentMask(None=TextureCoordinate() {coordinateIndex=2}) {g=true}
   Exp=ScalarParameter() {defaultValue=8, group=CloudSettings, parameterName=Cloud-HorizColorMaskPower}
  )
  Clamp(None=Multiply(A=@n1, B=ScalarParameter() {defaultValue=1, group=CloudSettings, parameterName=Cloud-HorizColorMaskMod})) {clampMode=CMODE_ClampMax}
  n2 = ScalarParameter() {defaultValue=38, group=ParentSettings, parameterName=Cloud-MainCloudTile}
  n3 = Multiply(A=ScalarParameter() {defaultValue=0.7, group=ParentSettings, parameterName=Cloud-TileMaster}, B=TextureCoordinate()) {constB=0.5}
  n4 = ScalarParameter() {defaultValue=1.5, group=ParentSettings, parameterName=Cloud-SpeedMaster}
  n5 = Multiply(A=ScalarParameter() {defaultValue=0.02, group=ParentSettings, parameterName=Cloud-MainCloudSpeed}, B=@n4)
  n6 = Multiply(A=Add(
    A=Multiply(A=@n2, B=@n3) {constB=0.5}
    B=MaterialFunctionCall(Speed=@n5) {materialFunction=&TimeWithSpeedVariable}.Time
   )) {constB=0.125}
  n7 = Constant() {r=1}
  n8 = ScalarParameter() {defaultValue=0.03, group=ParentSettings, parameterName=SwirlPanSpeed}
  n9 = MaterialFunctionCall!(
   SineAnimOffset
   Flow Direction=Constant2Vector() {r=-1}
   Time offset
   EdgeMaskSize
   UVS * size=Multiply(A=Subtract(A=@n3, B=@n7), B=ScalarParameter() {defaultValue=38, group=ParentSettings, parameterName=SwirlTile}) {constB=0.5}
   SineAnimSpeed=ScalarParameter() {defaultValue=1.5, group=ParentSettings, parameterName=SwirlCurlSpeed}
   Grid Noise Tile Rep
   Rotation amount modulator=ScalarParameter() {defaultValue=-0.5, group=ParentSettings, parameterName=SwirlAmount}
   un frac Time=MaterialFunctionCall(Speed=Multiply(A=@n8, B=@n4)) {materialFunction=&TimeWithSpeedVariable}.Time
  ) {desc=15 fewer instructions on low end, materialFunction=/Game/Packages/Fortress_Sky/Fortnite_SkyTwistFunction}
  n10 = ComponentMask(None=@n9.Rotated UVs) {desc=rotated uvs, g=true, r=true}
  n11 = Add(A=@n6, B=Multiply(A=@n10) {constB=0.125})
  n12 = Multiply(A=TextureCoordinate(), B=ScalarParameter() {defaultValue=4, group=MobileSettings, parameterName=MOBILE_CloudTile}) {constB=0.5}
  n13 = MaterialFunctionCall(Speed=ScalarParameter() {defaultValue=0.003, group=MobileSettings, parameterName=MOBILE_CloudSpeed}) {materialFunction=&TimeWithSpeedVariable}
  n14 = TextureSample(UVs=ShadingPathSwitch(Default=@n11, Mobile=Add(A=@n12, B=@n13.Time)) {desc=Migrated from FeatureLevelSwitch}) {
   samplerSource=SSM_Wrap_WorldGroupSettings
   samplerType=SAMPLERTYPE_LinearColor
   texture=/Game/Packages/Fortress_Sky/Materials/FinalCloudNormalAndHeightMap
  }
  ConstantBiasScale(None=@n14.RGB)
  Constant() {r=3}
  Constant() {r=4}
  Constant() {r=1.5}
  Constant() {r=3}
  Constant() {r=0.6}
  Constant() {r=-0.6}
  Constant() {r=1}
  Constant() {r=-0.48}
  n15 = CollectionParameter() {
   collection=&FortniteMaterialParameters
   parameterName=SunAndMoonModelWorldPosition
  }
  n16 = Normalize(VectorInput=Subtract(A=CameraPositionWS(), B=ComponentMask(None=@n15) {b=true, g=true, r=true}))
  n17 = DotProduct(A=CameraVectorWS(), B=@n16)
  n18 = ScalarParameter() {defaultValue=2000, group=SkySettings, parameterName=Sky-SunDiscPower}
  n19 = Clamp(None=Multiply(A=Power(Base=@n17, Exp=@n18), B=ScalarParameter() {defaultValue=2, group=SkySettings, parameterName=Sky-SunDiscMod})) {clampMode=CMODE_ClampMax}
  n20 = ScalarParameter() {defaultValue=25, group=SkySettings, parameterName=Sky-SunDiscPower2}
  n21 = VectorParameter() {group=Cloud-Colors, parameterName=Clouds-SunTintColor, defaultValue=<rgba 1 0.7527 0.5 1>}
  n22 = Multiply(A=Multiply(A=Power(Base=@n19, Exp=@n20), B=@n21.A), B=ScalarParameter() {defaultValue=10, group=SkySettings, parameterName=SunDiscIntensity})
  n23 = Clamp(None=@n17)
  n24 = Power(Base=@n23, Exp=ScalarParameter() {defaultValue=100, group=SkySettings, parameterName=Sky-SunGlowPower}) {constExponent=10}
  n25 = Multiply(A=@n24, B=Multiply(A=ScalarParameter() {defaultValue=0.1, group=SkySettings, parameterName=Sky-SunGlowMod}, B=@n23))
  n26 = ScalarParameter() {defaultValue=0.5, group=CloudSettings, parameterName=Cloud-OpacityMaskMod}
  n27 = ScalarParameter() {defaultValue=5, group=ParentSettings, parameterName=Cloud-DetailTextureTile}
  n28 = Time()
  n29 = Multiply(A=ScalarParameter() {defaultValue=-0.01, group=ParentSettings, parameterName=Cloud-DetailTextureSpeed}, B=@n4)
  n30 = TextureSample(UVs=Add(A=Multiply(A=@n11, B=@n27) {constA=3}, B=Multiply(A=@n28, B=@n29))) {samplerType=SAMPLERTYPE_LinearColor, texture=&FinalCloudNormalAndHeightMap}
  n31 = Subtract(A=OneMinus(None=@n30.B), B=ScalarParameter() {defaultValue=0.5, group=CloudSettings, parameterName=Cloud-OpacityMaskSub}) {constB=0.25}
  n32 = Add(A=@n14.B, B=Multiply(A=OneMinus(None=@n14.B), B=Multiply(A=@n26, B=@n31) {constA=0.25, constB=0.1}))
  n33 = Add(
   A=Power(Base=Clamp(None=@n32), Exp=ScalarParameter() {defaultValue=0.6, group=CloudSettings, parameterName=Cloud-BlobPow})
   B=ScalarParameter() {defaultValue=-0.6, group=CloudSettings, parameterName=Cloud-BlobSoftness}
  ) {constB=-0.5}
  n34 = VectorParameter() {group=CloudSettings, parameterName=Cloud-NormalsMod, defaultValue=<rgba 1 1 0.5 1>}
  n35 = Constant3Vector() {constant=<rgba 0 0 1 0>}
  n36 = Constant()
  n37 = ComponentMask(None=@n14.RGB) {g=true, r=true}
  n38 = Constant() {r=0.5}
  n39 = Add(A=Multiply(A=@n37, B=@n38), B=Multiply(A=@n38, B=ComponentMask(None=@n30.RGB) {g=true, r=true}))
  n40 = Subtract(A=ShadingPathSwitch(Default=@n32, Mobile=@n14.B) {desc=Migrated from FeatureLevelSwitch}, B=Constant() {r=0.5})
  n41 = ScalarParameter() {defaultValue=1.5, group=CloudSettings, parameterName=Cloud-OpacityMaskCombinePower}
  n42 = Clamp(None=Multiply(A=Power(Base=@n40, Exp=@n41), B=ScalarParameter() {defaultValue=10, group=CloudSettings, parameterName=Cloud-OpacityMaskCombineMod}))
  n43 = Multiply(A=ScalarParameter() {defaultValue=0.2, group=CloudSettings, parameterName=Cloud-DetailTextureNormalAmount}, B=@n42)
  n44 = DeriveNormalZ(InXY=ConstantBiasScale(None=LinearInterpolate(A=@n37, B=@n39, Alpha=@n43) {constAlpha=0.05}) {bias=-0.5, scale=2})
  n45 = Multiply(A=@n9.RotationAngleValues, B=Constant() {r=-1}) {constB=-1}
  n46 = Add(A=Multiply(A=RotateAboutAxis(NormalizedRotationAxis=@n35, RotationAngle=@n45, PivotPoint=@n36, Position=@n44) {desc=9}) {constB=0.5}, B=@n44)
  n47 = Transform(None=Normalize(VectorInput=Multiply(A=@n34.RGB, B=@n46)))
  n48 = ComponentMask(None=@n47) {b=true}
  n49 = ConstantBiasScale(None=@n48)
  n50 = Clamp(None=Multiply(A=ScalarParameter() {defaultValue=0.25, group=CloudSettings, parameterName=Cloud-BlobModA}, B=@n49) {constB=5})
  n51 = Add(A=@n3, B=Multiply(A=@n10, B=ScalarParameter() {defaultValue=0.01, group=ParentSettings, parameterName=Cloud-DetailTexture2Swirl}))
  n52 = ScalarParameter() {defaultValue=60, group=ParentSettings, parameterName=Cloud-DetailTexture2Tile}
  n53 = Time()
  n54 = Multiply(A=ScalarParameter() {defaultValue=-0.001, group=ParentSettings, parameterName=Cloud-DetailTexture2Speed}, B=@n4)
  n55 = TextureSample(UVs=Add(A=Multiply(A=@n51, B=@n52) {constA=3}, B=Multiply(A=@n53, B=@n54))) {samplerType=SAMPLERTYPE_LinearGrayscale, texture=/Game/Global/Textures/Default/Tileable_Patterns/T_StormCloudDetail01_M}
  n56 = Multiply(A=@n50, B=@n55.R)
  n57 = Add(A=@n50, B=@n55.R)
  n58 = Multiply(A=Clamp(None=Multiply(A=@n56, B=@n57)), B=ScalarParameter() {defaultValue=3, group=CloudSettings, parameterName=Cloud-BlobModB})
  n59 = Clamp(None=Add(A=@n42, B=Multiply(A=@n33, B=@n58)))
  n60 = Power(Base=@n49, Exp=ScalarParameter() {defaultValue=4, group=CloudSettings, parameterName=Cloud-SunScatterPowerB_1}) {constExponent=20}
  n61 = Multiply(
   A=Power(Base=OneMinus(None=@n55.R), Exp=ScalarParameter() {defaultValue=2.5, group=CloudSettings, parameterName=Cloud-SunScatterPowerA}) {constExponent=20}
   B=@n60
  )
  n62 = MaterialFunctionCall(Transition Phase=@n59, Height Texture=OneMinus(None=@n61)) {materialFunction=&HeightLerp}
  n63 = Add(A=@n59, B=Multiply(A=@n62.Lerp Alpha No Contrast, B=@n60))
  n64 = Constant() {r=0.25}
  n65 = ComponentMask(None=TextureCoordinate() {coordinateIndex=2}) {g=true}
  n66 = ScalarParameter() {defaultValue=3, group=CloudSettings, parameterName=Cloud-ZenithDissipatePower}
  n67 = Clamp(None=Multiply(A=Power(Base=@n65, Exp=@n66), B=ScalarParameter() {defaultValue=4, group=CloudSettings, parameterName=Cloud-ZenithDissapateMod})) {clampMode=CMODE_ClampMax}
  n68 = Multiply(
   A=ShadingPathSwitch(Default=Clamp(None=@n63) {clampMode=CMODE_ClampMax}, Mobile=Multiply(A=@n42, B=@n64)) {desc=Migrated from FeatureLevelSwitch}
   B=@n67
  )
  n69 = Multiply(
   A=Multiply(A=Add(A=@n22, B=@n25), B=@n21.RGB)
   B=Power(Base=OneMinus(None=@n68), Exp=ScalarParameter() {defaultValue=2, group=SkySettings, parameterName=Sky-SunGlowCloudMaskPower})
  )
  n70 = Add(
   A=Multiply(A=TextureCoordinate() {coordinateIndex=1}, B=ScalarParameter() {defaultValue=2.5, group=Stars, parameterName=Stars-GalaxyLargeTile})
   B=MaterialFunctionCall(X=Constant() {r=0.1}, Y=ScalarParameter() {defaultValue=-0.5, group=Stars, parameterName=Stars-GalaxyLargeOffset}) {materialFunction=&MakeFloat2}.Result
  )
  n71 = MaterialFunctionCall(Float2=@n70) {materialFunction=&BreakOutFloat2Components}
  n72 = Constant() {r=0.5}
  n73 = Constant()
  n74 = Multiply(
   A=Multiply(
    A=SphereMask(A=@n71.G, B=@n72, Radius=@n72, Hardness=@n73)
    B=TextureSample(UVs=@n70) {texture=/Game/Packages/Fortress_Sky/Textures/Galaxy_Background}.RGB
   )
   B=ScalarParameter() {defaultValue=0.3, group=Stars, parameterName=Stars-GalaxyLargeBright}
  )
  n75 = TextureCoordinate() {coordinateIndex=1}
  n76 = ScalarParameter() {defaultValue=6, group=Stars, parameterName=Stars-GalaxySmallTile}
  n77 = Multiply(
   A=TextureSample(UVs=Multiply(A=@n75, B=@n76)) {texture=/Game/Packages/Fortress_Sky/Textures/Galaxy_Background_Tiling}.RGB
   B=ScalarParameter() {defaultValue=0.2, group=Stars, parameterName=Stars-GalaxySmallBright}
  )
  n78 = TextureCoordinate() {coordinateIndex=1}
  n79 = ScalarParameter() {defaultValue=6, group=Stars, parameterName=Stars-StarTile}
  n80 = Power(
   Base=TextureSample(UVs=Multiply(A=@n78, B=@n79)) {samplerType=SAMPLERTYPE_LinearColor, texture=&/T_SkyDome_Stars_Sharp}.RGB
   Exp=ScalarParameter() {defaultValue=3, group=Stars, parameterName=Stars-StarPower}
  )
  n81 = ScalarParameter() {defaultValue=20, group=Stars, parameterName=Stars-StarMod}
  n82 = TextureSample(UVs=Multiply(A=@n78, B=ScalarParameter() {defaultValue=10, group=Stars, parameterName=Stars-StarMaskTileA})) {samplerType=SAMPLERTYPE_LinearGrayscale, texture=&T_StormCloudDetail01_M}
  n83 = ScalarParameter() {defaultValue=3, group=Stars, parameterName=Stars-StarMaskTileAPower}
  n84 = Panner(
   Coordinate=Multiply(A=@n78, B=ScalarParameter() {defaultValue=25, group=Stars, parameterName=Stars-StarMaskTileB})
   Time=MaterialFunctionCall(Speed=ScalarParameter() {defaultValue=0.5, group=Stars, parameterName=Stars-StarMaskTileBSpeed}) {materialFunction=&TimeWithSpeedVariable}.Time
  ) {speedX=0.1, speedY=0.1}
  n85 = Multiply(
   A=Multiply(A=Multiply(A=@n80, B=@n81), B=Power(Base=@n82.R, Exp=@n83))
   B=Power(
    Base=TextureSample(UVs=@n84) {samplerType=SAMPLERTYPE_LinearGrayscale, texture=&T_StormCloudDetail01_M}.R
    Exp=ScalarParameter() {defaultValue=4, group=Stars, parameterName=Stars-StarMaskTileBPower}
   )
  )
  n86 = ScalarParameter() {defaultValue=0.05, group=Stars, parameterName=Stars-StarBrightness}
  n87 = Multiply(
   A=Power(Base=OneMinus(None=@n68), Exp=ScalarParameter() {defaultValue=20, group=SkySettings, parameterName=Sky-CloudsOccludeStarsPower}) {constExponent=7}
   B=Add(A=Multiply(A=@n85, B=@n86), B=Add(A=@n74, B=@n77))
  )
  n88 = Constant()
  n89 = ComponentMask(None=TextureCoordinate() {coordinateIndex=2}) {g=true}
  n90 = LinearInterpolate(
   A=Add(
    A=ShadingPathSwitch(Default=@n87, Mobile=@n88) {desc=Migrated from FeatureLevelSwitch}
    B=VectorParameter() {group=Sky-Colors, parameterName=Sky-ZenithColor, defaultValue=<rgba 0 0 0 1>}.RGB
   )
   B=VectorParameter() {group=Sky-Colors, parameterName=Sky-HorizonColor, defaultValue=<rgba 0 0 0 1>}.RGB
   Alpha=Power(Base=@n89, Exp=ScalarParameter() {defaultValue=4, group=SkySettings, parameterName=Sky-MidHorizonPower})
  )
  n91 = ScalarParameter() {defaultValue=10, group=SkySettings, parameterName=Sky-LowHorizonPower}
  n92 = ScalarParameter() {defaultValue=2, group=SkySettings, parameterName=Sky-LowHorizonBrightness}
  n93 = Add(A=@n90, B=Multiply(A=Power(Base=@n89, Exp=@n91), B=Multiply(A=@n90, B=@n92)))
  n94 = Constant() {r=1}
  n95 = ScalarParameter() {defaultValue=0.75, group=CloudSettings, parameterName=Cloud-RimlightSharpPower}
  n96 = ScalarParameter() {defaultValue=100, group=CloudSettings, parameterName=Cloud-SunScatterModB}
  n97 = Multiply(A=@n67, B=Multiply(A=@n62.Lerp Alpha No Contrast, B=Multiply(A=@n61, B=@n96) {constA=0.25, constB=0.1}))
  n98 = Multiply(A=ScalarParameter() {defaultValue=10, group=CloudSettings, parameterName=Cloud-RimlightSunGlowMod}, B=@n24)
  n99 = Multiply(
   A=Power(Base=OneMinus(None=@n42), Exp=ScalarParameter() {defaultValue=8, group=CloudSettings, parameterName=Cloud-RimlightOutlinePower})
   B=Add(A=Power(Base=@n97, Exp=@n95), B=Multiply(A=@n97, B=@n98))
  )
  n100 = ScalarParameter() {defaultValue=50, group=CloudSettings, parameterName=Cloud-RimlightBrightness}
  n101 = ShadingPathSwitch(Default=Multiply(A=Multiply(A=@n99, B=@n100), B=@n23), Mobile=Constant()) {desc=Migrated from FeatureLevelSwitch}
  n102 = DotProduct(A=@n16, B=@n47)
  n103 = Power(
   Base=ConstantBiasScale(None=Multiply(A=@n102) {constB=-1})
   Exp=ScalarParameter() {defaultValue=2, group=CloudSettings, parameterName=Cloud-ShadowPower}
  ) {constExponent=20}
  n104 = ScalarParameter() {defaultValue=3, group=CloudSettings, parameterName=Cloud-ShadowMod}
  n105 = ShadingPathSwitch(Default=Clamp(None=Multiply(A=@n103, B=@n104) {constB=5}), Mobile=@n14.B) {desc=Migrated from FeatureLevelSwitch}
  n106 = LinearInterpolate(
   A=VectorParameter() {group=Cloud-Colors, parameterName=Cloud-ShadowColor, defaultValue=<rgba 0 0 0 1>}.RGB
   B=Multiply(
    A=VectorParameter() {group=Cloud-Colors, parameterName=Clouds-KeyLightColor, defaultValue=<rgba 0.4 0.39 0.36 1>}.RGB
    B=Add(A=@n94, B=@n101)
   )
   Alpha=Clamp(None=Add(A=@n101, B=@n105)) {clampMode=CMODE_ClampMax}
  )
  n107 = ConstantBiasScale(None=Multiply(A=@n48, B=Constant() {r=-1}) {constB=-1})
  n108 = ScalarParameter() {defaultValue=15, group=CloudSettings, parameterName=Cloud-RimlightPower_3}
  n109 = ShadingPathSwitch(
   Default=LinearInterpolate(
    A=@n106
    B=VectorParameter() {group=Cloud-Colors, parameterName=Clouds-BounceColor, defaultValue=<rgba 0.0845 0.1092 0.13 1>}.RGB
    Alpha=Power(Base=@n107, Exp=@n108) {constExponent=20}
   )
   Mobile=@n106
  ) {desc=Migrated from FeatureLevelSwitch}
  n110 = VectorParameter() {group=Cloud-Colors, parameterName=Clouds-Horizon Color, defaultValue=<rgba 0.12 0.5238 1.2 1>}
  Multiply(
   A=LinearInterpolate(A=Add(A=@n69, B=@n93), B=LinearInterpolate(A=@n109, B=@n110.RGB, Alpha=@n1), Alpha=@n68)
   B=ShadingPathSwitch(
    Default=ScalarParameter() {defaultValue=1, parameterName=DEFAULT_GlobalBrightness}
    Mobile=ScalarParameter() {defaultValue=1, group=MobileSettings, parameterName=MOBILE_GlobalBrightness}
   ) {desc=Migrated from FeatureLevelSwitch}
  )
  Multiply(A=@n9.Rotated UVs, B=@n9.Rotated UVs)
  Multiply(A=@n9.RotationAngleValues, B=@n9.RotationAngleValues)
  Reroute(None=@n60)
  Reroute(None=@n6)
 ]
}
