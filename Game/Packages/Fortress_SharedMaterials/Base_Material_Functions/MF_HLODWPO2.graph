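# Material function that adds an animated world-collapse effect to material attributes, apparently for HLOD meshes, behind a static switch.
# A looping, exponentially eased timer pushes vertices down harder with distance from CollapseCenter and pulls lower ones toward it, clamped against overshoot.
# A Custom node zeroes the offset at zero progress, and a time-varying radial mask fades emissive to black.
# An older suck-down chain remains unconnected.
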MaterialFunction /Game/Packages/Fortress_SharedMaterials/Base_Material_Functions/MF_HLODWPO2
graph {nodes [
  n1 = Add(B=WorldPosition().XYZ)
  n2 = Divide(
   A=Distance(A=ComponentMask() {g=true, r=true}, B=ComponentMask(None=@n1) {g=true, r=true})
   B=ScalarParameter() {defaultValue=1393645, parameterName=SuckDownRadius}
  )
  n3 = Power(Base=OneMinus(None=Saturate(None=@n2)), Exp=ScalarParameter() {defaultValue=16, parameterName=SuckDownExponent})
  n4 = Multiply(A=Distance(A=@n1), B=ScalarParameter() {defaultValue=-1, parameterName=SuckDownDistance})
  Add(B=AppendVector(A=Constant2Vector(), B=Multiply(A=@n3, B=@n4)))
  n5 = FunctionInput() {inputType=FunctionInput_MaterialAttributes}
  n6 = GetMaterialAttributes(None=@n5) {attributeGetTypes [#1, #2]}
  n7 = VectorParameter() {parameterName=CollapseCenter, defaultValue=<rgba 28821.1055 -14329.0605 0 1>}
  n8 = WorldPosition() {worldPositionShaderOffset=WPT_ExcludeAllShaderOffsets}
  n9 = Distance(A=ComponentMask(None=@n7.RGB) {g=true, r=true}, B=ComponentMask(None=@n8.XYZ) {g=true, r=true})
  n10 = Saturate(None=Divide(A=@n9, B=ScalarParameter() {defaultValue=1000000, parameterName=PushDown_Falloff}))
  n11 = ScalarParameter() {defaultValue=2, parameterName=PushDown_Exp}
  n12 = Constant()
  n13 = ScalarParameter() {defaultValue=3000000, parameterName=PushDown}
  n14 = Multiply(A=Time(), B=ScalarParameter() {defaultValue=0.05, parameterName=ColapseSpeed})
  n15 = Saturate(None=Multiply(A=Frac(None=@n14), B=ScalarParameter() {defaultValue=1, parameterName=PauseAdd}) {constB=2})
  n16 = Saturate(None=Power(Base=@n15, Exp=ScalarParameter() {defaultValue=8, parameterName=ExponentialTime}))
  n17 = Multiply(A=Constant3Vector() {constant=<rgba 0 0 -1 0>}, B=Multiply(A=Power(Base=@n10, Exp=@n11), B=LinearInterpolate(A=@n12, B=@n13, Alpha=@n16)))
  n18 = Subtract(A=@n7.RGB, B=AppendVector(A=Constant2Vector(), B=ScalarParameter() {parameterName=SkyHigh}))
  n19 = WorldPosition() {worldPositionShaderOffset=WPT_ExcludeAllShaderOffsets}
  n20 = Normalize(VectorInput=Subtract(A=@n18, B=@n19.XYZ))
  n21 = Clamp(
   None=LinearInterpolate(A=Constant(), B=ScalarParameter() {defaultValue=866000000, parameterName=SuckRadius_Max}, Alpha=@n16)
   Max=Distance(A=@n18, B=@n19.XYZ)
  ) {clampMode=CMODE_ClampMax}
  n22 = Subtract(A=ComponentMask(None=Add(A=@n17, B=@n19.XYZ)) {b=true}, B=ScalarParameter() {defaultValue=-500000, parameterName=Offset})
  n23 = ScalarParameter() {defaultValue=500000, parameterName=Fade}
  MaterialExpressionLinearInterpolate_8 = LinearInterpolate(A=Add(A=@n17, B=Multiply(A=@n20, B=@n21)), B=@n17, Alpha=Saturate(None=Divide(A=@n22, B=@n23)))
  MaterialExpressionReroute_0 = Reroute(None=@n15)
  n24 = Custom!(WPO=@MaterialExpressionLinearInterpolate_8, Progress=@MaterialExpressionReroute_0) {
   code="float3 newWPO = {0,0,0};\r\n\r\nif (Progress <= 0)\r\n{\r\n\treturn newWPO;\r\n}\r\nelse\r\n{\r\n\treturn WPO;\r\n}\r\n"
   inputs [
    {
     inputName=WPO
     input {expression=&self:MaterialExpressionLinearInterpolate_8}
    }
    {
     inputName=Progress
     input {expression=&self:MaterialExpressionReroute_0}
    }
   ]
  }
  n25 = Add(
   A=VertexInterpolator(VS=@n9).PS
   B=LinearInterpolate(
    A=ScalarParameter() {parameterName=EmissiveMaskOffset}
    B=ScalarParameter() {defaultValue=-4384796, parameterName=EmissiveMaskOffset2}
    Alpha=@n16
   )
  )
  n26 = LinearInterpolate(
   A=ScalarParameter() {defaultValue=5000000, parameterName=EmissiveMaskRadius}
   B=ScalarParameter() {defaultValue=50000000, parameterName=EmissiveMaskRadius2}
   Alpha=@n16
  )
  n27 = LinearInterpolate(A=Constant3Vector(), B=@n6.EmissiveColor, Alpha=Saturate(None=Divide(A=@n25, B=@n26)))
  FunctionOutput(None=StaticSwitchParameter(
    True=SetMaterialAttributes!(MaterialAttributes=@n6.MaterialAttributes, WorldPositionOffset=@n24, EmissiveColor=@n27) {attributeSetTypes [#1, #2]}
    False=@n5
   ) {parameterName=GlobalSwitchToEnableFunction})
  Multiply()
  Normalize(VectorInput=Multiply(A=Multiply(A=@n8.XYZ, B=@n7.RGB), B=Constant3Vector() {constant=<rgba 1 1 1 0>}))
 ]}
