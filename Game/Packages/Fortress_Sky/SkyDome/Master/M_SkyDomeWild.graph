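# Unlit sky dome master material drawing a gradient sky, stars, sun and moon behind animated, lit clouds.
# Collection parameters place the sun/moon texture on a virtual plane and drive glow and atmosphere sphere masks.
# Swirl-distorted main, detail and wispy layers form cloud opacity, with texture-derived normals providing sun lighting, near-sun spill, bounce and horizon wash.
# Mobile switches use flat-colored clouds, unswirled main-cloud UVs and non-panning stars.
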Material /Game/Packages/Fortress_Sky/SkyDome/Master/M_SkyDomeWild
graph {
 properties {bTangentSpaceNormal=false, bUseMaterialAttributes=true, pixelDepthOffsetMode=PDOM_Legacy, shadingModel=MSM_Unlit, usedShadingModels=Unlit}
 nodes [
  n1 = CollectionParameter() {
   collection=&FortniteMaterialParameters
   parameterName=SunAndMoonModelDirectionalVector
  }
  n2 = CollectionParameter() {
   collection=&FortniteMaterialParameters
   parameterName=SunlightYTransformVector
  }
  n3 = CollectionParameter() {
   collection=&FortniteMaterialParameters
   parameterName=SunAndMoonModelWorldPosition
  }
  n4 = ComponentMask(None=@n3) {b=true, g=true, r=true}
  n5 = TextureSample(UVs=MaterialFunctionCall(
    Plane Normal Axis=ComponentMask(None=@n1) {b=true, g=true, r=true}
    Plane X Axis=ComponentMask(None=@n2) {b=true, g=true, r=true}
    Plane Center=@n4
    UV Scale=ScalarParameter() {defaultValue=3000000, group=SunSettings, parameterName=SunMoon-Scale}
   ) {materialFunction=&VirtualPlaneCoordinates}.UVs) {samplerType=SAMPLERTYPE_LinearColor, texture=&/T_SunMoon}
  n6 = Constant() {r=2}
  n7 = SphereMask(
   A=WorldPosition().XYZ
   B=@n4
   Radius=ScalarParameter() {defaultValue=4500000, group=SunSettings, parameterName=SunMoon-AtmosphereMaskRadius}
   Hardness=Constant()
  )
  n8 = Multiply(
   A=Power(Base=@n7, Exp=ScalarParameter() {defaultValue=1, group=SunSettings, parameterName=SunMoon-GlowPower})
   B=ScalarParameter() {group=SunSettings, parameterName=SunMoon-GlowMod}
  )
  Add(A=Multiply(A=Power(Base=@n5.B, Exp=@n6), B=Constant() {r=0.5}), B=@n8)
  Constant2Vector() {r=-1}
  Constant() {r=6.5}
  Constant() {r=-0.003}
  Constant() {r=0.3}
  Constant() {r=14}
  Constant() {r=1.8}
  Constant() {r=0.03}
  Constant() {r=3}
  Constant() {r=1}
  Constant() {r=3}
  Constant() {r=15}
  Constant() {r=8}
  Constant() {r=2.8}
  Constant() {r=-0.02}
  n9 = Constant() {r=0.125}
  n10 = Subtract(
   A=Multiply(A=TextureCoordinate(), B=ScalarParameter() {defaultValue=1, group=Parent-MI-Settings, parameterName=Cloud-TileMaster}) {constB=0.5}
   B=Constant() {r=1}
  )
  n11 = ScalarParameter() {defaultValue=0.05, group=Parent-MI-Settings, parameterName=SwirlPanSpeed}
  n12 = ScalarParameter() {defaultValue=1, group=Parent-MI-Settings, parameterName=Cloud-SpeedMaster}
  n13 = MaterialFunctionCall!(
   SineAnimOffset
   Flow Direction=Constant2Vector() {g=0.1, r=-1}
   Time offset
   EdgeMaskSize
   UVS * size=Multiply(A=@n10, B=ScalarParameter() {defaultValue=30, group=Parent-MI-Settings, parameterName=SwirlTile}) {constB=0.5}
   SineAnimSpeed=ScalarParameter() {defaultValue=1.5, group=Parent-MI-Settings, parameterName=SwirlCurlSpeed}
   Grid Noise Tile Rep
   Rotation amount modulator=ScalarParameter() {defaultValue=0.4, group=Parent-MI-Settings, parameterName=SwirlAmount}
   un frac Time=MaterialFunctionCall(Speed=Multiply(A=@n11, B=@n12)) {materialFunction=&TimeWithSpeedVariable}.Time
  ) {desc=15 fewer instructions on low end, materialFunction=/Game/Packages/Fortress_Sky/Fortnite_SkyTwistFunction}
  n14 = ComponentMask(None=Multiply(A=@n13.Rotated UVs, B=@n13.Rotated UVs)) {desc=rotated uvs, g=true, r=true}
  n15 = Multiply(A=@n14, B=@n9) {constB=0.125}
  n16 = ScalarParameter() {defaultValue=14, group=Parent-MI-Settings, parameterName=Cloud-MainCloudTile}
  n17 = Multiply(A=ScalarParameter() {defaultValue=0.02, group=Parent-MI-Settings, parameterName=Cloud-MainCloudSpeed}, B=@n12)
  n18 = Multiply(
   A=Add(
    A=Multiply(A=@n16, B=@n10) {constB=0.5}
    B=MaterialFunctionCall(Speed=@n17) {materialFunction=&TimeWithSpeedVariable}.Time
   )
   B=@n9
  ) {constB=0.125}
  n19 = TextureSampleParameter2D(UVs=ShadingPathSwitch(Default=Add(A=@n15, B=@n18), Mobile=@n18) {desc=Migrated from FeatureLevelSwitch}) {
   parameterName=MainCloudTexture
   samplerType=SAMPLERTYPE_LinearColor
   texture=&/T_SkyDome_CloudShapes
  }
  n20 = Power(Base=@n19.B, Exp=ScalarParameter() {defaultValue=4, group=MobileSettings, parameterName=MOBILE-MainCloudShape-Power})
  n21 = Saturate(None=Multiply(A=ScalarParameter() {defaultValue=-100, group=SunSettings, parameterName=Moon-PreBlendOffset}, B=@n5.R))
  n22 = ScalarParameter() {defaultValue=500000000, group=SunSettings, parameterName=Sun-SharpnessMod}
  n23 = Power(Base=@n5.G, Exp=ScalarParameter() {defaultValue=150, group=SunSettings, parameterName=Sun-SharpnessPower})
  n24 = Multiply(A=ScalarParameter() {defaultValue=40, group=SunSettings, parameterName=Sun-Amount}, B=Saturate(None=Multiply(A=@n22, B=@n23)))
  n25 = Multiply(
   A=Saturate(None=Add(A=@n21, B=@n24))
   B=VectorParameter() {group=MobileSettings, parameterName=MOBILE-SunMoon-Color, defaultValue=<rgba 1 0.767 0.4688 1>}.RGB
  )
  n26 = ScalarParameter() {defaultValue=50, group=SunSettings, parameterName=SunMoon-UVErrorMaskMod}
  n27 = Saturate(None=Add(A=Multiply(A=@n7, B=@n26), B=ScalarParameter() {defaultValue=-20, group=SunSettings, parameterName=SunMoon-UVErrorMaskAdd}))
  n28 = ComponentMask(None=TextureCoordinate() {coordinateIndex=2}) {g=true}
  n29 = LinearInterpolate(
   A=VectorParameter() {group=Colors, parameterName=Sky-ZenithColor, defaultValue=<rgba 0.02 0.0859 0.2 1>}.RGB
   B=VectorParameter() {group=Colors, parameterName=Sky-HorizonColor, defaultValue=<rgba 0.25 1.0732 2.5 1>}.RGB
   Alpha=Power(Base=@n28, Exp=ScalarParameter() {defaultValue=4.5, group=Colors, parameterName=Sky-HorizonPower})
  )
  n30 = ScalarParameter() {defaultValue=1, group=Colors, parameterName=Sky-LowHorizon-AddBright}
  n31 = ScalarParameter() {defaultValue=10, group=Colors, parameterName=Sky-LowHorizon-Power}
  n32 = Add(A=@n29, B=Multiply(A=Multiply(A=@n29, B=@n30), B=Power(Base=@n28, Exp=@n31)))
  n33 = Saturate(None=@n32)
  n34 = VectorParameter() {group=MobileSettings, parameterName=MOBILE-SunMoon-AtmosphereColor, defaultValue=<rgba 0.3594 0.3523 0.2826 1>}
  n35 = ScalarParameter() {defaultValue=1, group=SunSettings, parameterName=SunMoon-AtmospherePower}
  n36 = Saturate(None=Multiply(A=Power(Base=@n7, Exp=@n35), B=ScalarParameter() {group=SunSettings, parameterName=SunMoon-AtmosphereMod}))
  n37 = Add(
   A=Multiply(A=@n20, B=VectorParameter() {group=MobileSettings, parameterName=MOBILE-CloudsColor, defaultValue=<rgba 0 0 0 1>}.RGB)
   B=Add(A=Multiply(A=@n25, B=@n27), B=LinearInterpolate(A=@n33, B=@n34.RGB, Alpha=@n36))
  )
  n38 = Multiply(A=@n14, B=Constant() {r=0.4})
  n39 = Multiply(A=@n10, B=ScalarParameter() {defaultValue=8, group=Parent-MI-Settings, parameterName=Wispy2-Tile}) {constB=0.5}
  n40 = MaterialFunctionCall(Speed=Multiply(A=@n12, B=ScalarParameter() {defaultValue=0.0007, group=Parent-MI-Settings, parameterName=Wispy2-Speed})) {materialFunction=&TimeWithSpeedVariable}
  n41 = TextureSampleParameter2D(UVs=Add(A=@n38, B=Add(A=@n39, B=@n40.Time))) {parameterName=CloudandStars, samplerType=SAMPLERTYPE_LinearColor, texture=&/T_SkyDome_StarCloudMask}
  n42 = Power(Base=Multiply(A=MaterialFunctionCall!(S=@n41.B, V2, V3, V4) {desc=Wispy Cloud, materialFunction=/Engine/Functions/Engine_MaterialFunctions02/PassThrough}.S)) {constExponent=3}
  n43 = ScalarParameter() {defaultValue=2, group=CloudOpacitySettings, parameterName=MainCloud-OpacityMaskMod}
  n44 = Power(Base=@n19.B, Exp=ScalarParameter() {defaultValue=4, group=CloudOpacitySettings, parameterName=MainCloud-OpacityMaskPower})
  n45 = ScalarParameter() {defaultValue=1, group=Cloud-ShapeSettings, parameterName=MainCloudShape-Mod}
  n46 = Power(Base=@n19.B, Exp=ScalarParameter() {defaultValue=1, group=Cloud-ShapeSettings, parameterName=MainCloudShape-Power})
  n47 = ScalarParameter() {defaultValue=1, group=Cloud-ShapeSettings, parameterName=DetailCloudShape-Mod}
  n48 = Multiply(A=ScalarParameter() {defaultValue=6.5, group=Parent-MI-Settings, parameterName=Cloud-DetailCloudTile}, B=@n10) {constB=0.5}
  n49 = MaterialFunctionCall(Speed=Multiply(A=ScalarParameter() {defaultValue=-0.003, group=Parent-MI-Settings, parameterName=Cloud-DetailCloudSpeed}, B=@n12)) {materialFunction=&TimeWithSpeedVariable}
  n50 = ScalarParameter() {defaultValue=0.3, group=Parent-MI-Settings, parameterName=Cloud-DetailCloudSwirl}
  n51 = TextureSampleParameter2D(UVs=Add(A=Add(A=@n48, B=@n49.Time), B=Multiply(A=@n50, B=@n14))) {
   parameterName=CloudDetailTexture
   samplerType=SAMPLERTYPE_LinearColor
   texture=&/T_SkyDome_CloudShapesSmall
  }
  n52 = Power(Base=@n51.B, Exp=ScalarParameter() {defaultValue=1, group=Cloud-ShapeSettings, parameterName=DetailCloudShape-Power})
  n53 = MaterialFunctionCall(Transition Phase=Saturate(None=Multiply(A=@n45, B=@n46)), Height Texture=Saturate(None=Multiply(A=@n47, B=@n52))) {materialFunction=&HeightLerp}
  n54 = LinearInterpolate(
   A=Saturate(None=Multiply(A=@n43, B=@n44))
   B=@n53.Lerp Alpha No Contrast
   Alpha=ScalarParameter() {defaultValue=0.6, group=CloudOpacitySettings, parameterName=OpacityLerpAmount}
  )
  n55 = Multiply(A=MaterialFunctionCall(S=@n54) {desc=Cloud Shape, materialFunction=&PassThrough}.S) {constB=3, desc=Blow out the values a bit to get those goopier Fortnite clouds.}
  n56 = VectorParameter() {group=SunSettings, parameterName=SunMoon-AtmosphereColor, defaultValue=<rgba 0 1 0.0176 1>}
  n57 = TextureCoordinate() {coordinateIndex=1}
  n58 = TextureSampleParameter2D(UVs=Multiply(A=@n57, B=ScalarParameter() {defaultValue=7, group=Stars, parameterName=StarTile})) {parameterName=CloudandStars, samplerType=SAMPLERTYPE_LinearColor, texture=&T_SkyDome_StarCloudMask}
  n59 = ScalarParameter() {defaultValue=15, group=Stars, parameterName=StarMaskTile}
  n60 = ScalarParameter() {defaultValue=0.05, group=Stars, parameterName=StarMaskSpeed}
  n61 = TextureSampleParameter2D(UVs=Panner(
    Coordinate=Multiply(A=@n57, B=@n59)
    Time=MaterialFunctionCall(Speed=@n60) {materialFunction=&TimeWithSpeedVariable}.Time
   ) {speedX=-1}) {parameterName=CloudandStars, samplerType=SAMPLERTYPE_LinearColor, texture=&T_SkyDome_StarCloudMask}
  n62 = Power(
   Base=ShadingPathSwitch(Default=Multiply(A=@n58.R, B=@n61.G), Mobile=Multiply(A=@n58.R, B=@n58.R)) {desc=Save 4 instructions on es3 by removing Star panning}
   Exp=ScalarParameter() {defaultValue=3, group=Stars, parameterName=StarPower}
  )
  n63 = ScalarParameter() {group=Stars, parameterName=StarBrightness}
  n64 = Add(A=Multiply(A=ScalarParameter() {group=SunSettings, parameterName=Moon-Amount}, B=@n21), B=@n24)
  n65 = VectorParameter() {group=SunSettings, parameterName=SunMoon-Color, defaultValue=<rgba 3 1.8818 0.45 1>}
  n66 = VectorParameter() {group=SunSettings, parameterName=SunMoon-GlowColor, defaultValue=<rgba 1 0.9505 0.9 1>}
  n67 = Multiply(A=Add(A=Multiply(A=@n64, B=@n65.RGB), B=Multiply(A=@n8, B=@n66.RGB)), B=@n27)
  n68 = Add(A=@n67, B=Add(A=LinearInterpolate(A=@n32, B=@n56.RGB, Alpha=@n36), B=Multiply(A=@n62, B=@n63)))
  n69 = LinearInterpolate(
   A=@n68
   B=@n37
   Alpha=Saturate(None=LinearInterpolate(A=@n42, Alpha=@n55) {constB=0.6, desc="Mix Cloud Base and wisps.\r\nTried a simple add and it makes the night too bright."})
  ) {desc="A is stars/background, B is the old mobile emissive"}
  n70 = ScalarParameter() {defaultValue=3, group=CloudOpacitySettings, parameterName=1-FinalCloudOpacityBoost}
  n71 = Add(A=Add(A=@n19.B, B=@n51.B), B=ScalarParameter() {defaultValue=-3, group=CloudOpacitySettings, parameterName=CloudOpacityAdd}) {constB=-0.5}
  n72 = Constant() {r=-1}
  n73 = Multiply(A=@n13.RotationAngleValues, B=@n13.RotationAngleValues)
  n74 = Constant() {r=0.5}
  n75 = Multiply(A=@n19.RGB, B=@n74)
  n76 = Multiply(A=@n74, B=@n51.RGB)
  n77 = Add(
   A=ScalarParameter() {defaultValue=0.2, group=Cloud-ShapeSettings, parameterName=DetailCloudShape-NegOne-to-One}
   B=@n53.Lerp Alpha No Contrast
  )
  n78 = ComponentMask(None=LinearInterpolate(A=Add(A=@n75, B=@n76), B=@n19.RGB, Alpha=Saturate(None=@n77))) {g=true, r=true}
  n79 = Multiply(
   A=DeriveNormalZ(InXY=ConstantBiasScale(None=@n78) {bias=-0.5, scale=2})
   B=VectorParameter() {group=Cloud-ShapeSettings, parameterName=Cloud-NormalsMod, defaultValue=<rgba 1 1 1 1>}.RGB
  )
  n80 = Multiply(A=RotateAboutAxis(
    NormalizedRotationAxis=Constant3Vector() {constant=<rgba 0 0 1 0>}
    RotationAngle=Multiply(A=@n72, B=@n73) {constB=-1}
    PivotPoint=Constant()
    Position=@n79
   ) {desc=9}) {constB=0.5}
  n81 = Transform(None=Normalize(VectorInput=Add(A=@n80, B=@n79)))
  n82 = ComponentMask(None=@n81) {b=true}
  n83 = ConstantBiasScale(None=@n82)
  n84 = ScalarParameter() {defaultValue=1, group=CloudOpacitySettings, parameterName=Detail-LerpA-ModWispy1}
  n85 = ScalarParameter() {group=CloudOpacitySettings, parameterName=Detail-LerpB-ModWispy1}
  n86 = Multiply(A=ScalarParameter() {defaultValue=15, group=Parent-MI-Settings, parameterName=Wispy1-Tile}, B=@n10) {constB=0.5}
  n87 = MaterialFunctionCall(Speed=Multiply(A=@n12, B=ScalarParameter() {defaultValue=0.001, group=Parent-MI-Settings, parameterName=Wispy1-Speed})) {materialFunction=&TimeWithSpeedVariable}
  n88 = TextureSampleParameter2D(UVs=Add(A=@n38, B=Add(A=@n86, B=@n87.Time))) {parameterName=CloudandStars, samplerType=SAMPLERTYPE_LinearColor, texture=&T_SkyDome_StarCloudMask}
  n89 = Saturate(None=Multiply(A=LinearInterpolate(A=@n84, B=@n85, Alpha=@n51.B), B=@n88.B))
  n90 = Multiply(A=@n83, B=@n89)
  n91 = Add(A=@n83, B=@n89)
  n92 = Multiply(A=Saturate(None=Multiply(A=@n90, B=@n91)), B=ScalarParameter() {defaultValue=0.2, group=CloudOpacitySettings, parameterName=CloudOpacityMod})
  n93 = Saturate(None=Add(A=Multiply(A=@n71, B=@n92), B=@n54))
  n94 = ScalarParameter() {defaultValue=1, group=Cloud-LightingMasks, parameterName=NearSun-SpilloverLight-Power}
  n95 = Saturate(None=Multiply(A=Power(Base=@n83, Exp=@n94), B=ScalarParameter() {defaultValue=0.1, group=Cloud-LightingMasks, parameterName=NearSun-SpilloverLight-Mod}))
  n96 = ScalarParameter() {defaultValue=15, group=Cloud-LightingMasks, parameterName=NearSun-CloudOpacityPower}
  n97 = OneMinus(None=@n93)
  n98 = Saturate(None=Multiply(A=Power(Base=@n97, Exp=@n96), B=ScalarParameter() {defaultValue=50, group=Cloud-LightingMasks, parameterName=NearSun-CloudOpacityMod}))
  n99 = CollectionParameter() {
   collection=&FortniteMaterialParameters
   parameterName=SunAndMoonModelWorldPosition
  }
  n100 = ComponentMask(None=@n99) {b=true, g=true, r=true}
  n101 = SphereMask(
   A=WorldPosition().XYZ
   B=@n100
   Radius=ScalarParameter() {defaultValue=3500000, group=Cloud-LightingMasks, parameterName=HemisphereLight-Size}
   Hardness=Constant()
  )
  n102 = ScalarParameter() {defaultValue=2, group=Cloud-LightingMasks, parameterName=HemisphereLight-Power}
  n103 = Saturate(None=Multiply(A=Power(Base=@n101, Exp=@n102), B=ScalarParameter() {defaultValue=1, group=Cloud-LightingMasks, parameterName=HemisphereLight-Mod}))
  n104 = Multiply(
   A=Multiply(A=Add(A=@n95, B=@n98), B=ScalarParameter() {defaultValue=5, group=Cloud-LightingMasks, parameterName=NearSun-LightAmount})
   B=Power(Base=@n103, Exp=ScalarParameter() {defaultValue=5, group=Cloud-LightingMasks, parameterName=NearSun-HemisphereLight-MaskPower})
  )
  n105 = ScalarParameter() {defaultValue=50, group=Cloud-LightingMasks, parameterName=Sun-HemisphereShadowMod}
  n106 = ScalarParameter() {defaultValue=1, group=Cloud-LightingMasks, parameterName=Lighting-CloudOpacityAffectsLighting}
  n107 = Saturate(None=Add(A=Multiply(A=@n105, B=@n103), B=Power(Base=@n93, Exp=@n106)))
  n108 = Constant()
  n109 = Normalize(VectorInput=Subtract(A=CameraPositionWS(), B=@n100))
  n110 = ConstantBiasScale(None=Multiply(A=DotProduct(A=@n81, B=@n109), B=Constant() {r=-1}) {constB=-1})
  n111 = Multiply(
   A=Power(Base=@n110, Exp=ScalarParameter() {defaultValue=3, group=Cloud-LightingMasks, parameterName=Lighting-ShadowPower}) {constExponent=20}
   B=ScalarParameter() {defaultValue=0.8, group=Cloud-LightingMasks, parameterName=Lighting-Brightness}
  ) {constB=5}
  n112 = OneMinus(None=@n110)
  n113 = ScalarParameter() {defaultValue=-0.3, group=Cloud-LightingMasks, parameterName=Wispy2-ShadowBlend}
  n114 = ScalarParameter() {defaultValue=0.3, group=Cloud-LightingMasks, parameterName=Wispy2-HeightAdd}
  n115 = Saturate(None=MaterialFunctionCall(A=@n111, B=Multiply(A=@n111, B=@n108), Transition Phase=Add(A=@n112, B=@n113), Height Texture=Add(A=@n41.B, B=@n114)) {materialFunction=&HeightLerp}.Results)
  n116 = LinearInterpolate(
   A=VectorParameter() {group=Colors, parameterName=Cloud-ShadowColor, defaultValue=<rgba 0.06 0.1088 0.2 1>}.RGB
   B=VectorParameter() {group=Colors, parameterName=Cloud-DirectSunLightColor, defaultValue=<rgba 3 3 3 1>}.RGB
   Alpha=Add(A=Multiply(A=@n104, B=@n93), B=Multiply(A=@n107, B=@n115))
  )
  n117 = VectorParameter() {group=Colors, parameterName=Cloud-BounceColorZenith, defaultValue=<rgba 0.1 0.2346 0.5 1>}
  n118 = VectorParameter() {group=Colors, parameterName=Cloud-BounceColorHorizon, defaultValue=<rgba 0.4 0.9385 2 1>}
  n119 = ComponentMask(None=TextureCoordinate() {coordinateIndex=2}) {g=true}
  n120 = Power(Base=@n119, Exp=ScalarParameter() {defaultValue=4, group=Cloud-LightingMasks, parameterName=Bounce-HorizonColorMask-Power})
  n121 = Add(A=Multiply(A=@n82, B=Constant() {r=-1}) {constB=-1}, B=Constant() {r=-1})
  n122 = Multiply(
   A=Power(Base=ConstantBiasScale(None=@n121), Exp=ScalarParameter() {defaultValue=5, group=Cloud-LightingMasks, parameterName=Bounce-MaskPower})
   B=ScalarParameter() {defaultValue=20, group=Cloud-LightingMasks, parameterName=Bounce-MaskMod}
  ) {constB=-1}
  n123 = Multiply(
   A=Power(Base=OneMinus(None=@n103), Exp=ScalarParameter() {defaultValue=1, group=Cloud-LightingMasks, parameterName=Bounce-HemisphereMaskPower})
   B=Multiply(A=LinearInterpolate(A=@n117.RGB, B=@n118.RGB, Alpha=@n120), B=Saturate(None=@n122))
  )
  n124 = Multiply(
   A=ScalarParameter() {defaultValue=1.5, group=Colors, parameterName=Cloud-HorizColorWash-Amount}
   B=Power(Base=@n119, Exp=ScalarParameter() {defaultValue=10, group=Colors, parameterName=Cloud-HorizColorWash-Power})
  )
  n125 = LinearInterpolate(
   A=@n68
   B=LinearInterpolate(
    A=Add(A=@n116, B=@n123)
    B=VectorParameter() {group=Colors, parameterName=Cloud-HorizonColorWash, defaultValue=<rgba 0.25 1.0732 2.5 1>}.RGB
    Alpha=Saturate(None=@n124)
   )
   Alpha=Saturate(None=Multiply(A=@n93, B=@n70))
  )
  MaterialFunctionCall!(In=MakeMaterialAttributes!(
    BaseColor
    Metallic
    Specular
    Roughness
    Anisotropy
    EmissiveColor=ShadingPathSwitch(Default=@n125, Mobile=@n69) {desc=Migrated from FeatureLevelSwitch}
    Opacity
    OpacityMask
    Normal
    Tangent
    WorldPositionOffset
    SubsurfaceColor
    ClearCoat
    ClearCoatRoughness
    AmbientOcclusion
    Refraction
    CustomizedUVs_0
    CustomizedUVs_1
    CustomizedUVs_2
    CustomizedUVs_3
    CustomizedUVs_4
    CustomizedUVs_5
    CustomizedUVs_6
    CustomizedUVs_7
    PixelDepthOffset
    ShadingModel
    Displacement
   )) {materialFunction=/Game/Packages/Fortress_SharedMaterials/Base_Material_Functions/MF_HLODWPO2}
  n126 = Constant()
  n127 = Constant() {r=5000000}
  Multiply(
   A=MaterialFunctionCall(
    Plane Normal Axis=Constant3Vector() {constant=<rgba 0 0 1 0>}
    Plane X Axis=Constant3Vector() {constant=<rgba 1 0.0015 0 1>}
    Plane Center=MaterialFunctionCall(X=@n126, Y=@n126, Z=@n127) {materialFunction=&MakeFloat3}.Result
    UV Scale=Constant() {r=100000000}
   ) {materialFunction=&VirtualPlaneCoordinates}.UVs
   B=Constant2Vector() {g=-1, r=1}
  )
  Multiply(A=@n20, B=ScalarParameter() {defaultValue=0.5, group=MobileSettings, parameterName=MOBILE-CloudAmount})
  Reroute(None=@n37)
  Reroute(None=Multiply(A=Constant(), B=@n67))
  Reroute(None=@n13.RotationAngleValues)
  Reroute(None=@n13.Rotated UVs)
  ScalarParameter() {defaultValue=3, desc=Makes main cloud shapes thicker, group=MobileSettings, parameterName=MOBILE-CloudShapeMult, sortPriority=9}
  ScalarParameter() {defaultValue=1, desc=Make wispy clouds more prominent, group=MobileSettings, parameterName=MOBILE-CloudWispMult, sortPriority=9}
  ScalarParameter() {defaultValue=3, desc="Breaks the wispy clouds up ", group=MobileSettings, parameterName=MOBILE-CloudWispPower, sortPriority=9}
  ScalarParameter() {
   defaultValue=0.6
   desc=How see-through the main cloud shapes are
   group=MobileSettings
   parameterName=MOBILE-CloudShapeMaxOpacity
   sortPriority=9
  }
 ]
}
